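# Translucent UI material for the UIKit input field background, drawing a rounded backplate that reacts to widget state.
# It samples the Slate post buffer, blends it partway toward the CoreUI Backplate B treatment, and derives a brighter hover variant.
# The shared modular-button state function selects the color, using one input for hovered, focused and pressed and a separate disabled backplate.
# A rounded-mask function takes corner radius parameters to shape the field.
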Material /FortUILibrary/Systems/InputField/Materials/M_UIKit_InputField_Background
graph {
 properties {
  bUsedWithStaticMesh=false
  blendMode=BLEND_Translucent
  materialDomain=MD_UI
  pixelDepthOffsetMode=PDOM_Legacy
  usedShadingModels=Default Lit
 }
 nodes [
  n1 = VectorParameter() {parameterName=Corner Radius, defaultValue=<rgba 20 20 20 20>}
  MaterialFunctionCall(Corner Radius=AppendVector(A=@n1.RGB, B=@n1.RGB), Radius in Pixels=ScalarParameter() {defaultValue=1, parameterName=Radius in Pixels}) {materialFunction=&MF_UI_CoreUI_RoundedMask}
  n2 = NamedRerouteUsage()
  n3 = MaterialFunctionCall!(Post Buffer Blur=@n2) {materialFunction=/FortUILibrary/MaterialFunctions/CoreUI/MF_UI_CoreUI_BackplateB}
  NamedRerouteDeclaration(None=Saturate(None=LinearInterpolate(A=@n2, B=@n3.Result) {constAlpha=0.4})) {name=B Weak, nodeColor=<rgba 0 0 0 1>}
  n4 = NamedRerouteUsage()
  n5 = Saturate(None=MaterialFunctionCall!(Base=@n4, Add=Constant3Vector() {constant=<rgba 0.7135 0.7135 0.7135 1>}) {materialFunction=/Engine/Functions/Engine_MaterialFunctions02/Math/MultiplyAdd}.Result)
  NamedRerouteDeclaration(None=Saturate(None=LinearInterpolate(A=@n4, B=@n5))) {name=B Weak Hover, nodeColor=<rgba 0 0 0 1>}
  NamedRerouteDeclaration(None=MaterialFunctionCall!("UVs (Optional)") {materialFunction=/Engine/Functions/UserInterface/GetSlatePost0}.LinearRGB) {name=Buffer RGB, nodeColor=<rgba 0 0 0 1>}
  n6 = NamedRerouteUsage()
  n7 = NamedRerouteUsage()
  NamedRerouteDeclaration(None=MaterialFunctionCall!(
    Idle=NamedRerouteUsage()
    Idle Opacity
    Hovered=@n6
    Hovered Opacity
    Focused=@n6
    Focused Opacity
    Pressed=@n6
    Pressed Opacity
    Disabled=MaterialFunctionCall!(Post Buffer Blur=@n7) {materialFunction=/FortUILibrary/MaterialFunctions/CoreUI/MF_UI_CoreUI_BackplateDisabled}.Result
    Disabled Opacity
   ) {materialFunction=/FortUILibrary/Systems/ModularButton/MaterialFunctions/MF_UIKit_ButtonStatesBase_Lerp}.Color) {name=Visual State, nodeColor=<rgba 0 0 0 1>}
  NamedRerouteUsage()
 ]
}
