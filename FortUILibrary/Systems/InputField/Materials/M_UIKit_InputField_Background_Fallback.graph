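# Translucent UI material that serves as a fallback background for the UIKit input field.
# A rounded-mask function, driven by Corner Radius and Radius in Pixels parameters, supplies the shape, which is multiplied to half strength.
# Two hard-coded muted lavender colors, B Weak and B Weak Hover, are declared as named reroutes.
# The modular button state-lerp function outputs the Visual State color, with Hovered, Focused and Pressed sharing one input.
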Material /FortUILibrary/Systems/InputField/Materials/M_UIKit_InputField_Background_Fallback
graph {
 properties {
  bUsedWithStaticMesh=false
  blendMode=BLEND_Translucent
  materialDomain=MD_UI
  pixelDepthOffsetMode=PDOM_Legacy
  usedShadingModels=Default Lit
 }
 nodes [
  n1 = VectorParameter() {parameterName=Corner Radius, defaultValue=<rgba 20 20 20 20>}
  Multiply(A=MaterialFunctionCall(Corner Radius=AppendVector(A=@n1.RGB, B=@n1.RGB), Radius in Pixels=ScalarParameter() {defaultValue=1, parameterName=Radius in Pixels}) {materialFunction=&MF_UI_CoreUI_RoundedMask}.Shape Mask) {constB=0.5}
  NamedRerouteDeclaration(None=Constant3Vector() {constant=<rgba 0.3864 0.3763 0.4793 1>}) {name=B Weak, nodeColor=<rgba 0 0 0 1>}
  NamedRerouteDeclaration(None=Constant3Vector() {constant=<rgba 0.7084 0.6867 0.8714 1>}) {name=B Weak Hover, nodeColor=<rgba 0 0 0 1>}
  n2 = NamedRerouteUsage()
  NamedRerouteDeclaration(None=MaterialFunctionCall!(
    Idle=NamedRerouteUsage()
    Idle Opacity
    Hovered=@n2
    Hovered Opacity
    Focused=@n2
    Focused Opacity
    Pressed=@n2
    Pressed Opacity
    Disabled=NamedRerouteUsage()
    Disabled Opacity
   ) {materialFunction=/FortUILibrary/Systems/ModularButton/MaterialFunctions/MF_UIKit_ButtonStatesBase_Lerp}.Color) {name=Visual State, nodeColor=<rgba 0 0 0 1>}
  NamedRerouteUsage()
 ]
}
