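# Translucent UI material for an input field icon, tinted and faded by the field's interaction state.
# MF_UIKit_ButtonStatesBase_Lerp supplies state color and opacity, which blend toward a hint-text color and then toward the disabled color, so disabled takes priority.
# The Icon texture serves only as a mask: the larger of red and blue (green is unused) times alpha scales the blended opacity.
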Material /FortUILibrary/Systems/InputField/Materials/M_UIKit_InputField_Icon
graph {
 properties {
  bUsedWithStaticMesh=false
  blendMode=BLEND_Translucent
  materialDomain=MD_UI
  pixelDepthOffsetMode=PDOM_Legacy
  usedShadingModels=Default Lit
 }
 nodes [
  n1 = VectorParameter() {group=State colors, parameterName=Color_Disabled, sortPriority=4, defaultValue=<rgba 1 1 1 1>}
  n2 = MaterialFunctionCall!(
   Idle
   Idle Opacity
   Hovered
   Hovered Opacity
   Focused
   Focused Opacity
   Pressed
   Pressed Opacity
   Disabled=NamedRerouteDeclaration(None=@n1.RGB) {name=Disabled RGB, nodeColor=<rgba 0 0 0 1>}
   Disabled Opacity=NamedRerouteDeclaration(None=@n1.A) {name=Disabled A, nodeColor=<rgba 0 0 0 1>}
  ) {materialFunction=/FortUILibrary/Systems/ModularButton/MaterialFunctions/MF_UIKit_ButtonStatesBase_Lerp}
  n3 = VectorParameter() {group=State colors, parameterName=Color_HintText, defaultValue=<rgba 0.5776 0.5776 0.5776 1>}
  n4 = ScalarParameter() {group=State transitions, parameterName=Trans_HintText, sliderMax=1, sortPriority=1}
  n5 = ScalarParameter() {group=State transitions, parameterName=Trans_Disabled, sliderMax=1, sortPriority=4}
  LinearInterpolate(A=LinearInterpolate(A=@n2.Color, B=@n3.RGB, Alpha=@n4) {desc=Color}, B=NamedRerouteUsage(), Alpha=@n5)
  n6 = LinearInterpolate(A=@n2.Opacity, B=@n3.A, Alpha=@n4) {desc=Opacity}
  n7 = NamedRerouteUsage()
  n8 = TextureSampleParameter2D() {parameterName=Icon}
  n9 = Max(A=@n8.R, B=@n8.R)
  Multiply(A=Multiply(A=LinearInterpolate(A=@n6, B=@n7, Alpha=@n5), B=Max(A=@n9, B=@n8.B)), B=@n8.A)
 ]
}
