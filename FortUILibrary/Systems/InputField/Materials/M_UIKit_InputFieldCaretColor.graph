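# Translucent UI-domain material for the input field caret color in the UIKit input field system.
# It reads the Color parameter from the MPC_UIKit_InputFieldCaretColor material parameter collection, so the value comes from shared data outside the material.
# A component mask takes the alpha channel of that color.
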Material /FortUILibrary/Systems/InputField/Materials/M_UIKit_InputFieldCaretColor
graph {
 properties {
  bUsedWithStaticMesh=false
  blendMode=BLEND_Translucent
  materialDomain=MD_UI
  pixelDepthOffsetMode=PDOM_Legacy
  usedShadingModels=Default Lit
 }
 nodes [ComponentMask(None=CollectionParameter() {collection=&/MPC_UIKit_InputFieldCaretColor, parameterName=Color}) {a=true}]
}
