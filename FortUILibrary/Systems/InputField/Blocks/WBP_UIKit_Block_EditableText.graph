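# Editable text block for the UIKit input field, built on the modular button block base around a single UIKitEditableText.
# Setter functions expose text, hint, font and input options, with font changes caching the dynamic font material.
# Hover, focus, press and disabled animations and an empty-text check drive that material's transition parameters.
# Gaining focus copies Color_Focused into a shared parameter collection for caret color, defaulting to white without a cached material.
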WidgetBlueprint /FortUILibrary/Systems/InputField/Blocks/WBP_UIKit_Block_EditableText
animations [
 Anim_Hovered_Transition {
  binding_count=1
  length_seconds=1
  object=&self:Anim_Hovered_Transition
  sequence {bindings [EditableText {
     id=#1
     possessed_class=UIKitEditableText
     tracks [{
       type=MovieSceneWidgetMaterialTrack
       display_name=Style.Font.FontMaterial
       sections [{
         type=MovieSceneParameterSection
         end_frame=20
         start_frame=0
         channels [Trans_Hovered {type=MovieSceneScriptingFloatChannel, keys [[0,0],[20,1]]}]
        }]
      }]
    }]}
 }
 Anim_Focused_Transition {
  binding_count=1
  length_seconds=1
  object=&self:Anim_Focused_Transition
  sequence {bindings [EditableText {
     id=#2
     possessed_class=UIKitEditableText
     tracks [{
       type=MovieSceneWidgetMaterialTrack
       display_name=Style.Font.FontMaterial
       sections [{
         type=MovieSceneParameterSection
         end_frame=20
         start_frame=0
         channels [Trans_Focused {type=MovieSceneScriptingFloatChannel, keys [[0,0],[20,1]]}]
        }]
      }]
    }]}
 }
 Anim_Pressed_Transition {
  binding_count=1
  length_seconds=1
  object=&self:Anim_Pressed_Transition
  sequence {bindings [EditableText {
     id=#3
     possessed_class=UIKitEditableText
     tracks [{
       type=MovieSceneWidgetMaterialTrack
       display_name=Style.Font.FontMaterial
       sections [{
         type=MovieSceneParameterSection
         end_frame=20
         start_frame=0
         channels [Trans_Pressed {type=MovieSceneScriptingFloatChannel, keys [[0,0],[20,1]]}]
        }]
      }]
    }]}
 }
 Anim_Disabled_Transition {
  binding_count=1
  length_seconds=1
  object=&self:Anim_Disabled_Transition
  sequence {bindings [EditableText {
     id=#4
     possessed_class=UIKitEditableText
     tracks [{
       type=MovieSceneWidgetMaterialTrack
       display_name=Style.Font.FontMaterial
       sections [{
         type=MovieSceneParameterSection
         end_frame=20
         start_frame=0
         channels [Trans_Disabled {type=MovieSceneScriptingFloatChannel, keys [[0,0],[20,1]]}]
        }]
      }]
    }]}
 }
]
graphs [
 EventGraph {nodes [
   n1 = "OnTextChanged (Editable Text) (EditableText)"!() -> (OutputDelegate, then, Text:"Text (by ref)") {is=title K2Node_ComponentBoundEvent}
   > SetFilledState!(execute, self, IsEmpty:bool=TextIsEmpty(InText=@n1.Text)) -> (then)
   Event OnAddedToFocusPath!() -> (OutputDelegate, then, InFocusEvent:Focus Event Structure) {is=title K2Node_Event}
   > n2 = Get!(execute:Exec) -> (then:Exec, else:Exec, FontDynamicMaterial:Material Instance Dynamic Object Reference)
   > n3 = Get Vector Parameter Value!(execute, self:Material Instance Dynamic Object Reference=@n2, ParameterName:Name=Color_Focused) -> (then, ReturnValue:Linear Color Structure) {is=title K2Node_CallFunction}
   > SetVectorParameterValue!(
    execute
    Collection:Material Parameter Collection Object Reference=/FortUILibrary/Systems/InputField/Materials/MPC_UIKit_InputFieldCaretColor
    ParameterName:Name=Color
    ParameterValue:"Linear Color Structure (by ref)"=@n3
   ) -> (then) {is=title K2Node_CallMaterialParameterCollectionFunction}
   SetVectorParameterValue!(
    execute=@n2.else
    Collection=&MPC_UIKit_InputFieldCaretColor
    ParameterName=Color
    ParameterValue_R:float=1
    ParameterValue_G:float=1
    ParameterValue_B:float=1
    ParameterValue_A:float=1
   ) -> (then)
   Event PreConstruct()
   > Set RenderOpacity!(execute, RenderOpacity:float=3, self:Widget Object Reference=Get EditableText() -> (EditableText:UIKit Editable Text Object Reference)) -> (then, Output_Get:float)
   > DelayUntilNextTick!(execute) -> (then)
   n4 = Get EditableText()
   > SetFilledState(IsEmpty=TextIsEmpty(InText=Get Text(self:Editable Text Object Reference=@n4)))
  ]}
 SetClearKeyboardFocusOnCommitBehavior {nodes [
   n1 = SetClearKeyboardFocusOnCommitBehavior!() -> (then, ShouldClearKeyboardFocusOnCommit:bool=~) {is=title K2Node_FunctionEntry}
   > Set ClearKeyboardFocusOnCommit!(execute, ClearKeyboardFocusOnCommit:bool=@n1, self:Editable Text Object Reference=Get EditableText()) -> (then, Output_Get:bool)
  ]}
 SetEnableIntegratedKeyboard {nodes [
   n1 = SetEnableIntegratedKeyboard!() -> (then, EnableIntegratedKeyboard:bool=~) {is=title K2Node_FunctionEntry}
   > Set EnableIntegratedKeyboard!(execute, EnableIntegratedKeyboard:bool=@n1, self:Editable Text Object Reference=Get EditableText()) -> (then, Output_Get:bool)
  ]}
 SetFilledState {nodes [
   n1 = SetFilledState!() -> (then, IsEmpty:bool=~) {is=title K2Node_FunctionEntry}
   > n2 = GetDynamicFontMaterial!(execute, self:UIKit Editable Text Object Reference=Get EditableText()) -> (then, ReturnValue:Material Instance Dynamic Object Reference)
   > SetScalarParameterValue(self=@n2, ParameterName=Trans_HintText, Value=Select(Option 0:float=0, Option 1:float=1, Index=@n1) -> (ReturnValue:float))
  ]}
 SetFont {nodes [
   n1 = SetFont!() -> (then, Font:Slate Font Info Structure) {is=title K2Node_FunctionEntry}
   n2 = Get EditableText()
   > SetFont(self:Editable Text Object Reference=@n2, InFontInfo=@n1) {is=title K2Node_CallFunction}
   > n3 = GetDynamicFontMaterial(self=@n2)
   > Set FontDynamicMaterial(FontDynamicMaterial:Material Instance Dynamic Object Reference=@n3) -> (Output_Get:Material Instance Dynamic Object Reference)
  ]}
 SetFontMaterial {nodes [
   n1 = SetFontMaterial!() -> (then, Material Instance:Material Instance Object Reference) {is=title K2Node_FunctionEntry}
   n2 = Get EditableText()
   > SetFontMaterial!(execute, self:Editable Text Object Reference=@n2, InMaterial:Material Interface Object Reference=@n1) -> (then) {is=title K2Node_CallFunction}
   > n3 = GetDynamicFontMaterial(self=@n2)
   > Set FontDynamicMaterial(FontDynamicMaterial=@n3)
  ]}
 SetHintText {nodes [
   n1 = SetHintText!() -> (then, Hint Text:Text) {is=title K2Node_FunctionEntry}
   > SetHintText!(execute, self:Editable Text Object Reference=Get EditableText(), InHintText:Text=@n1) -> (then) {is=title K2Node_CallFunction}
  ]}
 SetIsPassword {nodes [
   n1 = SetIsPassword!() -> (then, IsPassword:bool=~) {is=title K2Node_FunctionEntry}
   n2 = Get EditableText()
   > n3 = Set IsPassword!(execute, IsPassword:bool=@n1, self:Editable Text Object Reference=@n2) -> (then, Output_Get:bool)
   > SetIsPassword!(execute, self:Editable Text Object Reference=@n2, InbIsPassword:bool=@n3) -> (then) {is=title K2Node_CallFunction}
  ]}
 SetJustification {nodes [
   n1 = SetJustification!() -> (then, Justification:ETextJustify Enum=Left) {is=title K2Node_FunctionEntry}
   > Set Justification!(execute, Justification:ETextJustify Enum=@n1, self:Editable Text Object Reference=Get EditableText()) -> (then, Output_Get:ETextJustify Enum=Left)
  ]}
 SetMinDesiredWidth {nodes [
   n1 = SetMinDesiredWidth!() -> (then, Min Width:double=~) {is=title K2Node_FunctionEntry}
   n2 = Get EditableText()
   > n3 = Set MinimumDesiredWidth!(execute, MinimumDesiredWidth:float=@n1, self:Editable Text Object Reference=@n2) -> (then, Output_Get:float)
   > SetMinimumDesiredWidth!(execute, self:Editable Text Object Reference=@n2, InMinDesiredWidth:float=@n3) -> (then)
  ]}
 SetMobileFontSizeMultiplier {nodes [
   n1 = SetMobileFontSizeMultiplier!() -> (then, MobileFontSizeMultiplier:float=~) {is=title K2Node_FunctionEntry}
   > SetMobileFontSizeMultiplier!(execute, self:UIKit Editable Text Object Reference=Get EditableText(), InMobileFontSizeMultiplier:float=@n1) -> (then) {is=title K2Node_CallFunction}
  ]}
 SetReadOnly {nodes [
   n1 = SetReadOnly!() -> (then, IsReadOnly:bool=~) {is=title K2Node_FunctionEntry}
   n2 = Get EditableText()
   > n3 = Set IsReadOnly!(execute, IsReadOnly:bool=@n1, self:Editable Text Object Reference=@n2) -> (then, Output_Get:bool)
   > "SetIsReadOnly (Editable Text)"!(execute, self:Editable Text Object Reference=@n2, InbIsReadyOnly:bool=@n3) -> (then)
  ]}
 SetText {nodes [
   n1 = SetText!() -> (then, Text:Text) {is=title K2Node_FunctionEntry}
   > "SetText (Editable Text)"!(execute, self:Editable Text Object Reference=Get EditableText(), InText:Text=@n1) -> (then)
  ]}
]
variables [FontDynamicMaterial {type=MaterialInstanceDynamic}]
widgets {
 info {
  inheritedWidgetCount=0
  parentClass=/FortUILibrary/Systems/ModularButton/Blocks/WBP_UIKit_Block_Base.WBP_UIKit_Block_Base_C
  rootWidgetClass=/Script/FortniteUI.UIKitEditableText
  widgetCount=1
 }
 widgets [EditableText {
   widgetClassPath=&UIKitEditableText
   bIsVariable=true
   widget=EditableText
   properties {
    clipping=ClipToBoundsAlways
    hintText=Hint...
    text=Editable Text
    virtualKeyboardDismissAction=TextCommitOnAccept
    virtualKeyboardOptions {bEnableAutocorrect=true}
    widgetStyle {
     backgroundImageComposing {
      bIsDynamicallyLoaded=false
      drawAs=NoDrawType
      imageType=FullColor
      mirroring=NoMirror
      resourceName=../../../Engine/Content/Slate/Old/HyperlinkDotted.png
      resourceObject=null
      tiling=Both
      imageSize=<0 0>
      margin=<ltrb 0 0 0 0.1875>
      outlineSettings {
       bUseBrushTransparency=false
       roundingType=HalfHeightRadius
       width=0
       color {colorUseRule=UseColor_Specified, specifiedColor=<rgba 0 0 0 0>}
       cornerRadii=<0 0 0 0>
      }
      tintColor {colorUseRule=UseColor_Specified, specifiedColor=<rgba 1 1 1 1>}
      uVRegion {bIsValid=false, max=<0 0>, min=<0 0>}
     }
     backgroundImageSelected {
      bIsDynamicallyLoaded=false
      drawAs=Box
      imageType=FullColor
      mirroring=NoMirror
      resourceName=../../../Engine/Content/Slate/Common/EditableTextSelectionBackground.png
      resourceObject=null
      tiling=NoTile
      imageSize=<0 0>
      margin=<ltrb 0.25 0.25 0.25 0.25>
      outlineSettings {
       bUseBrushTransparency=false
       roundingType=HalfHeightRadius
       width=0
       color {colorUseRule=UseColor_Specified, specifiedColor=<rgba 0 0 0 0>}
       cornerRadii=<0 0 0 0>
      }
      tintColor {colorUseRule=UseColor_Specified, specifiedColor=<rgba 1 1 1 1>}
      uVRegion {bIsValid=false, max=<0 0>, min=<0 0>}
     }
     caretImage {
      bIsDynamicallyLoaded=false
      drawAs=Image
      imageType=FullColor
      mirroring=NoMirror
      resourceName=../../../Engine/Content/Slate/Old/White.png
      resourceObject=/FortUILibrary/Systems/InputField/Materials/M_UIKit_InputFieldCaretColor
      tiling=NoTile
      imageSize=<16 16>
      margin=<ltrb 0 0 0 0>
      outlineSettings {
       bUseBrushTransparency=false
       roundingType=HalfHeightRadius
       width=0
       color {colorUseRule=UseColor_Specified, specifiedColor=<rgba 0 0 0 0>}
       cornerRadii=<0 0 0 0>
      }
      tintColor {colorUseRule=UseColor_Specified, specifiedColor=<rgba 1 1 1 1>}
      uVRegion {bIsValid=false, max=<0 0>, min=<0 0>}
     }
     colorAndOpacity {colorUseRule=UseColor_Specified, specifiedColor=<rgba 1 1 1 1>}
     font {
      bForceMonospaced=false
      bMaterialIsStencil=false
      fontMaterial=/FortUILibrary/Systems/InputField/MaterialInstances/MI_UIKit_InputFieldText
      fontObject=&HeadingNow
      letterSpacing=0
      monospacedWidth=1
      size=18
      skewAmount=0
      typefaceFontName=74 Regular
      outlineSettings {
       bApplyOutlineToDropShadows=false
       bMiteredCorners=false
       bSeparateFillAlpha=false
       outlineMaterial=null
       outlineSize=0
       outlineColor=<rgba 0 0 0 1>
      }
     }
     hintColorAndOpacity {colorUseRule=UseColor_Foreground, specifiedColor=<rgba 1 0 1 1>}
     selectedBackgroundColor {colorUseRule=UseColor_Foreground, specifiedColor=<rgba 1 0 1 1>}
    }
   }
  }]
}
