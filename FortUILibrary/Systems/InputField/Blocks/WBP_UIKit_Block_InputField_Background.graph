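# UI Kit background block for input fields, derived from the modular button background base, that shows focus, press, disabled and selection states on a single image.
# PreConstruct applies the background material and rounded corners, caching a dynamic material instance.
# Transition events play animations forward or reverse on material scalars, while instant enable/disable events stop animations and set them directly.
# Selection reuses the focus animation; selectable fields skip plain focus and press animations, using SelectionIsAvailable to gate Selected requests.
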WidgetBlueprint /FortUILibrary/Systems/InputField/Blocks/WBP_UIKit_Block_InputField_Background
animations [
 Anim_Focused_Transition {
  binding_count=1
  length_seconds=1
  object=&self:Anim_Focused_Transition
  sequence {
   bindings [Backing {
     id=#1
     possessed_class=Image
     tracks [{
       type=MovieSceneWidgetMaterialTrack
       sections [{
         type=MovieSceneParameterSection
         end_frame=120
         start_frame=0
         channels [Trans_Focused {type=MovieSceneScriptingFloatChannel, keys [[0,0],[120,1]]}]
        }]
      }]
    }]
   display_rate {numerator=120}
  }
 }
 Anim_Pressed_Transition {
  binding_count=1
  length_seconds=1
  object=&self:Anim_Pressed_Transition
  sequence {bindings [Backing {
     id=#2
     possessed_class=Image
     tracks [{
       type=MovieSceneWidgetMaterialTrack
       sections [{
         type=MovieSceneParameterSection
         end_frame=20
         start_frame=0
         channels [Trans_Pressed {type=MovieSceneScriptingFloatChannel, keys [[0,0],[19.9997,1]]}]
        }]
      }]
    }]}
 }
 Anim_Disabled_Transition {
  binding_count=1
  length_seconds=1
  object=&self:Anim_Disabled_Transition
  sequence {bindings [Backing {
     id=#3
     possessed_class=Image
     tracks [{
       type=MovieSceneWidgetMaterialTrack
       sections [{
         type=MovieSceneParameterSection
         end_frame=20
         start_frame=0
         channels [Trans_Disabled {type=MovieSceneScriptingFloatChannel, keys [[0,0],[20,1]]}]
        }]
      }]
    }]}
 }
]
graphs [
 EventGraph {nodes [
   Event OnTransitionSelectedRequested() {is=title K2Node_Event}
   > PlayTransition!(
    execute
    self:UIKit Block Object Reference
    Animation:Widget Animation Object Reference=Get Anim_Focused_Transition() -> (Anim_Focused_Transition:Widget Animation Object Reference)
    bPlayForward:bool=true
   ) -> (then)
   > Set SelectionIsAvailable(SelectionIsAvailable:bool=false) -> (Output_Get:bool)
   Event OnTransitionFocusedRequested() {is=title K2Node_Event}
   > n1 = Branch(Condition=GetIsSelectable!() -> (IsSelectable:bool))
   > Branch(Condition=GetIsSelected!() -> (IsSelected:bool))
   > Branch(Condition=Get SelectionIsAvailable() -> (SelectionIsAvailable:bool))
   > RequestTransition!(execute, self:UIKit Block Object Reference, State:EUIKitBlockVisualState Enum=Selected) -> (then)
   PlayTransition(execute=@n1.else, Animation=Get Anim_Focused_Transition(), bPlayForward=true)
   Event OnTransitionUnfocusedRequested() {is=title K2Node_Event}
   > n2 = Branch(Condition=GetIsSelectable())
   PlayTransition(execute=@n2.else, Animation=Get Anim_Focused_Transition(), bPlayForward=false)
   Event OnTransitionPressedRequested() {is=title K2Node_Event}
   > n3 = Branch(Condition=GetIsSelectable())
   PlayTransition(execute=@n3.else, Animation=Get Anim_Pressed_Transition() -> (Anim_Pressed_Transition:Widget Animation Object Reference), bPlayForward=true)
   Event OnTransitionReleasedRequested() {is=title K2Node_Event}
   > n4 = Branch(Condition=GetIsSelectable())
   PlayTransition(execute=@n4.else, Animation=Get Anim_Pressed_Transition(), bPlayForward=false)
   Event OnTransitionDisabledRequested() {is=title K2Node_Event}
   > PlayTransition(Animation=Get Anim_Disabled_Transition() -> (Anim_Disabled_Transition:Widget Animation Object Reference), bPlayForward=true)
   Event OnTransitionEnabledRequested() {is=title K2Node_Event}
   > PlayTransition(Animation=Get Anim_Disabled_Transition(), bPlayForward=false)
   Event PreConstruct()
   > SetBackgroundMaterial!(
    execute
    self:WBP UIKit Block Background Base Object Reference
    Material Instance:"Material Instance Object Reference (by ref)"=Get BackgroundMaterial() -> (BackgroundMaterial:Material Instance Object Reference)
   ) -> (then)
   > SetCornerRadius!(execute, self, Radius in Pixels:bool=false, TopLeft:double=0.3, TopRight:double=0.3, BottomRight:double=0.3, BottomLeft:double=0.3) -> (then)
   n5 = Event OnTransitionDeselectedFocusedRequested() {is=title K2Node_Event}
   PlayTransition(execute=[@n5.then, @n6.then], Animation=Get Anim_Focused_Transition(), bPlayForward=false)
   > SetScalarParameterValue(
    self=Get BackgroundDynamicMaterial() -> (BackgroundDynamicMaterial:Material Instance Dynamic Object Reference)
    ParameterName=Trans_Disabled
    Value=0
   )
   > Set SelectionIsAvailable(SelectionIsAvailable=true)
   n6 = Event OnTransitionDeselectedIdleRequested() {is=title K2Node_Event}
   Event OnInstantTransitionToEnabled() {is=title K2Node_Event}
   > StopAllAnimations(self:User Widget Object Reference)
   n7 = Get BackgroundDynamicMaterial()
   > SetScalarParameterValue(self=@n7, ParameterName=Trans_Hovered, Value=0)
   > SetScalarParameterValue(self=@n7, ParameterName=Trans_Focused, Value=0)
   > SetScalarParameterValue(self=@n7, ParameterName=Trans_Pressed, Value=0)
   > SetScalarParameterValue(self=@n7, ParameterName=Trans_Disabled, Value=0)
   Event OnInstantTransitionToDisabled() {is=title K2Node_Event}
   > StopAllAnimations()
   > SetScalarParameterValue(self=Get BackgroundDynamicMaterial(), ParameterName=Trans_Disabled, Value=1)
  ]}
 SetBackgroundBrush {nodes [
   n1 = SetBackgroundBrush!() -> (then, InBrush:"Slate Brush Structure (by ref)") {is=title K2Node_FunctionEntry}
   n2 = Get Backing() -> (Backing:Image Object Reference)
   > SetBrush(self=@n2, InBrush=@n1)
   > n3 = GetDynamicMaterial(self=@n2)
   > Set BackgroundDynamicMaterial(BackgroundDynamicMaterial:Material Instance Dynamic Object Reference=@n3) -> (Output_Get:Material Instance Dynamic Object Reference)
   > Return Node()
  ]}
 SetBackgroundMaterial {nodes [
   n1 = SetBackgroundMaterial!() -> (then, Material Instance:"Material Instance Object Reference (by ref)") {is=title K2Node_FunctionEntry}
   n2 = Get Backing()
   > SetBrushFromMaterial(self=@n2, Material=@n1)
   > n3 = GetDynamicMaterial(self=@n2)
   > Set BackgroundDynamicMaterial(BackgroundDynamicMaterial=@n3)
   > Return Node()
  ]}
 SetCornerRadius {nodes [
   n1 = SetCornerRadius!() -> (then, Radius in Pixels:bool=~, TopLeft:double=~, TopRight:double=~, BottomRight:double=~, BottomLeft:double=~) {is=title K2Node_FunctionEntry}
   n2 = Get BackgroundDynamicMaterial()
   > SetScalarParameterValue(self=@n2, ParameterName=Radius in Pixels, Value="To Float (Boolean)"(InBool=@n1))
   > SetVectorParameterValue(
    self=@n2
    ParameterName=Corner Radius
    Value=Make Linear Color!(R:float=@n1.TopLeft, G:float=@n1.BottomLeft, B:float=@n1.TopRight, A:float=@n1.BottomRight) -> (LinearColor:Linear Color Structure) {is=title K2Node_MakeStruct}
   )
  ]}
]
variables [SelectionIsAvailable {type=boolean}]
widgets {
 info {
  inheritedWidgetCount=0
  parentClass=/FortUILibrary/Systems/ModularButton/Blocks/WBP_UIKit_BlockBackground_Base.WBP_UIKit_BlockBackground_Base_C
  rootWidgetClass=&GridPanel
  widgetCount=2
 }
 widgets [
  Root {widgetClassPath=&GridPanel, widget=Root, properties {columnFill [1], rowFill [1]}}
  Backing {
   widgetClassPath=&Image
   bIsVariable=true
   parent=Root
   widget=Backing
   properties {
    visibility=HitTestInvisible
    brush {
     bIsDynamicallyLoaded=false
     drawAs=Image
     imageType=NoImage
     mirroring=NoMirror
     resourceName=None
     resourceObject=null
     tiling=NoTile
     imageSize=<32 32>
     margin=<ltrb 0.5 0.5 0.5 0.5>
     outlineSettings {
      bUseBrushTransparency=false
      roundingType=HalfHeightRadius
      width=0
      color {colorUseRule=UseColor_Specified, specifiedColor=<rgba 0 0 0 0>}
      cornerRadii=<0 0 0 0>
     }
     tintColor {colorUseRule=UseColor_Specified, specifiedColor=<rgba 1 1 1 1>}
     uVRegion {bIsValid=false, max=<0 0>, min=<0 0>}
    }
   }
   slot {class=GridSlot, properties {layer=3}}
  }
 ]
}
